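# UI material function for Reboot Rally sidebar that draws a texture as a floating, bobbing item, outputting RGB and Alpha.
# Time at base plus modifier speed forms a sine-eased triangle wave between Range Min and Max, optionally added to vertical offset.
# UI UVs are center-scaled with optional X flip, offset, fit to texture size and rotated.
# The clamped sample's alpha passes through edge-pixel clipping to form the image mask.
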MaterialFunction /Game/UI/Sidebar/RebootRally/Materials/MF_UI_RebootRally_FloatingItem
graph {nodes [
  FunctionOutput(None=NamedRerouteUsage()) {outputName=RGB}
  FunctionOutput(None=NamedRerouteUsage()) {outputName=Alpha, sortPriority=1}
  n1 = FunctionInput(Preview=TextureObject()) {bUsePreviewValueAsDefault=true, inputName=Texture, inputType=FunctionInput_Texture2D}
  NamedRerouteDeclaration(None=TextureProperty(Tex=@n1)) {name="Texture  Size", nodeColor=<rgba 0 0 0 1>}
  n2 = Constant() {r=-0.25}
  n3 = Constant() {r=0.25}
  n4 = FunctionInput(Preview=Constant() {r=0.25}) {bUsePreviewValueAsDefault=true, inputName=Base Speed, inputType=FunctionInput_Scalar, sortPriority=10}
  n5 = FunctionInput(Preview=Constant() {r=0.05}) {bUsePreviewValueAsDefault=true, inputName=Speed Modifier, inputType=FunctionInput_Scalar, sortPriority=11}
  n6 = Frac(None=Multiply(A=Time(), B=Add(A=@n4, B=@n5) {constB=0.05}) {constB=0.25})
  n7 = Multiply(A=Min(A=@n6, B=OneMinus(None=@n6))) {constB=2}
  NamedRerouteDeclaration(None=LinearInterpolate(
    A=FunctionInput(Preview=@n2) {bUsePreviewValueAsDefault=true, inputName=Range Min, inputType=FunctionInput_Scalar, sortPriority=12}
    B=FunctionInput(Preview=@n3) {bUsePreviewValueAsDefault=true, inputName=Range Max, inputType=FunctionInput_Scalar, sortPriority=13}
    Alpha=MaterialFunctionCall("Value (0-1)"=@n7) {materialFunction=&MF_UI_EaseCurves}.EaseInOutSine
   ) {constA=-0.25, constB=0.25}) {name=Offset Modifier, nodeColor=<rgba 0 0 0 1>}
  n8 = NamedRerouteDeclaration(None=MaterialFunctionCall() {materialFunction=&GetUserInterfaceUV}.Normalized UV) {name=UV, nodeColor=<rgba 0 0 0 1>}
  n9 = FunctionInput(Preview=Constant() {r=0.5}) {bUsePreviewValueAsDefault=true, inputName=Scale, inputType=FunctionInput_Scalar, sortPriority=2}
  n10 = MaterialFunctionCall(Float2=@n9) {materialFunction=&BreakOutFloat2Components}
  n11 = StaticSwitch(
   True=AppendVector(A=Multiply(A=@n10.R) {constB=-1}, B=@n10.G)
   False=@n9
   Value=FunctionInput(Preview=StaticBool() {value=true}) {bUsePreviewValueAsDefault=true, inputName=Flip X, inputType=FunctionInput_StaticBool, sortPriority=3}
  )
  n12 = MaterialFunctionCall(Float2=FunctionInput(Preview=Constant2Vector()) {bUsePreviewValueAsDefault=true, inputName=Offset, inputType=FunctionInput_Vector2, sortPriority=1}) {materialFunction=&BreakOutFloat2Components}
  n13 = StaticBool() {value=true}
  n14 = Add(
   A=@n12.G
   B=StaticSwitch(
    True=NamedRerouteUsage()
    False=Constant()
    Value=FunctionInput(Preview=@n13) {bUsePreviewValueAsDefault=true, inputName=Animate, inputType=FunctionInput_StaticBool, sortPriority=32}
   )
  )
  n15 = MaterialFunctionCall!(
   UV=Subtract(
    A=MaterialFunctionCall(UVs=@n8, Texture Scale=@n11) {materialFunction=&ScaleUVsByCenter}.UVs
    B=AppendVector(A=@n12.R, B=@n14)
   )
   TextureSize=NamedRerouteUsage()
  ) {materialFunction=&MF_UI_UVScaling}
  n16 = OneMinus(None=FunctionInput(Preview=Constant()) {bUsePreviewValueAsDefault=true, inputName=Rotation, inputType=FunctionInput_Scalar, sortPriority=4})
  n17 = TextureSample(
   UVs=NamedRerouteDeclaration(None=MaterialFunctionCall(UVs=@n15.Scaled to Fit, "Rotation Angle (0-1)"=@n16) {materialFunction=&CustomRotator}.Rotated Values) {name=Image UV, nodeColor=<rgba 0 0 0 1>}
   Tex=@n1
  ) {samplerSource=SSM_Clamp_WorldGroupSettings}
  NamedRerouteDeclaration(None=@n17.RGB) {name=Image RGB, nodeColor=<rgba 0 0 0 1>}
  NamedRerouteDeclaration(None=MaterialFunctionCall!(UVs=NamedRerouteUsage(), Alpha=@n17.A) {materialFunction=/FortUILibrary/MaterialFunctions/Basics/MF_UI_ClipEdgePixel}.Result) {name=Image Mask, nodeColor=<rgba 0 0 0 1>}
 ]}
