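# Translucent UI material for the Reboot Rally sidebar that composites up to five reward images into one animated picture.
# Each image runs through MF_UI_RebootRally_FloatingItem with its own texture, offset, scale, flip and rotation, plus varied bob speeds and a shared Animate switch.
# Color layers by chained lerps over a black edge color, while opacity takes the max of masks.
# Static switches toggle the fourth texture and an additional item.
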Material /Game/UI/Sidebar/RebootRally/Materials/M_UI_Reward_Composition
graph {
 properties {
  bUsedWithStaticMesh=false
  blendMode=BLEND_Translucent
  materialDomain=MD_UI
  pixelDepthOffsetMode=PDOM_Legacy
  usedShadingModels=Default Lit
 }
 nodes [
  n1 = Max(A=Max(A=NamedRerouteUsage(), B=NamedRerouteUsage()), B=NamedRerouteUsage())
  n2 = Max(A=@n1, B=NamedRerouteUsage())
  NamedRerouteDeclaration(None=StaticSwitchParameter(True=@n2, False=@n1) {parameterName=Display Texture 4}) {name=Opacity Pass, nodeColor=<rgba 0 0 0 1>}
  n3 = Constant3Vector() {desc=Edge Color, constant=<rgba 0 0 0 1>}
  n4 = LinearInterpolate(A=@n3, B=NamedRerouteUsage(), Alpha=NamedRerouteUsage())
  n5 = LinearInterpolate(
   A=LinearInterpolate(A=StaticSwitchParameter(True=@n4, False=@n3) {parameterName=Display Additionnal Item}, B=NamedRerouteUsage(), Alpha=NamedRerouteUsage())
   B=NamedRerouteUsage()
   Alpha=NamedRerouteUsage()
  )
  n6 = NamedRerouteUsage()
  n7 = NamedRerouteUsage()
  NamedRerouteDeclaration(None=LinearInterpolate(A=LinearInterpolate(A=@n5, B=@n6, Alpha=@n7), B=NamedRerouteUsage(), Alpha=NamedRerouteUsage())) {name=Color Pass, nodeColor=<rgba 0 0 0 1>}
  n8 = MaterialFunctionCall!(
   Texture=TextureObjectParameter() {group=Image 01 Properties, parameterName=Texture 01, sortPriority=0}
   Offset=AppendVector(
    A=ScalarParameter() {group=Image 01 Properties, parameterName=Image 01 Offset X}
    B=ScalarParameter() {group=Image 01 Properties, parameterName=Image 01 Offset Y}
   )
   Scale=ScalarParameter() {defaultValue=0.5, group=Image 01 Properties, parameterName=Image 01 Scale}
   Flip X=StaticBoolParameter() {group=Image 01 Properties, parameterName=Image 01 Flip X}
   Rotation=ScalarParameter() {group=Image 01 Properties, parameterName=Image 01 Rotation}
   Base Speed=NamedRerouteUsage()
   Speed Modifier=Constant()
   Range Min=NamedRerouteUsage()
   Range Max=NamedRerouteUsage()
   Animate=StaticBoolParameter() {defaultValue=true, parameterName=Animate}
  ) {materialFunction=&/MF_UI_RebootRally_FloatingItem}
  NamedRerouteDeclaration(None=@n8.RGB) {name=Image 01 RGB, nodeColor=<rgba 0 0 0 1>}
  NamedRerouteDeclaration(None=Constant() {r=0.2}) {name=Base Speed, nodeColor=<rgba 0 0 0 1>}
  NamedRerouteDeclaration(None=Constant() {r=-0.05}) {name=Range Offset Min, nodeColor=<rgba 0 0 0 1>}
  NamedRerouteDeclaration(None=Constant() {r=0.05}) {name=Range Offset Max, nodeColor=<rgba 0 0 0 1>}
  n9 = MaterialFunctionCall(
   Texture=TextureObjectParameter() {group=Image 05 Properties, parameterName=Texture 05, sortPriority=0}
   Offset=AppendVector(
    A=ScalarParameter() {group=Image 05 Properties, parameterName=Image 05 Offset X}
    B=ScalarParameter() {group=Image 05 Properties, parameterName=Image 05 Offset Y}
   )
   Scale=ScalarParameter() {defaultValue=0.5, group=Image 05 Properties, parameterName=Image 05 Scale}
   Flip X=StaticBoolParameter() {group=Image 05 Properties, parameterName=Image 05 Flip X}
   Rotation=ScalarParameter() {group=Image 05 Properties, parameterName=Image 05 Rotation}
   Base Speed=NamedRerouteUsage()
   Speed Modifier=Constant() {r=-0.025}
   Range Min=NamedRerouteUsage()
   Range Max=NamedRerouteUsage()
   Animate=StaticBoolParameter() {defaultValue=true, parameterName=Animate}
  ) {materialFunction=&MF_UI_RebootRally_FloatingItem}
  NamedRerouteDeclaration(None=@n9.RGB) {name=Image 05 RGB, nodeColor=<rgba 0 0 0 1>}
  NamedRerouteDeclaration(None=@n9.Alpha) {name=Image 05 Mask, nodeColor=<rgba 0 0 0 1>}
  n10 = MaterialFunctionCall(
   Texture=TextureObjectParameter() {group=Image 02 Properties, parameterName=Texture 02, sortPriority=0}
   Offset=AppendVector(
    A=ScalarParameter() {group=Image 02 Properties, parameterName=Image 02 Offset X}
    B=ScalarParameter() {group=Image 02 Properties, parameterName=Image 02 Offset Y}
   )
   Scale=ScalarParameter() {defaultValue=0.5, group=Image 02 Properties, parameterName=Image 02 Scale}
   Flip X=StaticBoolParameter() {group=Image 02 Properties, parameterName=Image 02 Flip X}
   Rotation=ScalarParameter() {group=Image 02 Properties, parameterName=Image 02 Rotation}
   Base Speed=NamedRerouteUsage()
   Speed Modifier=Constant() {r=0.05}
   Range Min=NamedRerouteUsage()
   Range Max=NamedRerouteUsage()
   Animate=StaticBoolParameter() {defaultValue=true, parameterName=Animate}
  ) {materialFunction=&MF_UI_RebootRally_FloatingItem}
  NamedRerouteDeclaration(None=@n10.RGB) {name=Image 02 RGB, nodeColor=<rgba 0 0 0 1>}
  n11 = MaterialFunctionCall(
   Texture=TextureObjectParameter() {group=Image 03 Properties, parameterName=Texture 03, sortPriority=0}
   Offset=AppendVector(
    A=ScalarParameter() {group=Image 03 Properties, parameterName=Image 03 Offset X}
    B=ScalarParameter() {group=Image 03 Properties, parameterName=Image 03 Offset Y}
   )
   Scale=ScalarParameter() {defaultValue=0.5, group=Image 03 Properties, parameterName=Image 03 Scale}
   Flip X=StaticBoolParameter() {group=Image 03 Properties, parameterName=Image 03 Flip X}
   Rotation=ScalarParameter() {group=Image 03 Properties, parameterName=Image 03 Rotation}
   Base Speed=NamedRerouteUsage()
   Speed Modifier=Constant() {r=-0.05}
   Range Min=NamedRerouteUsage()
   Range Max=NamedRerouteUsage()
   Animate=StaticBoolParameter() {defaultValue=true, parameterName=Animate}
  ) {materialFunction=&MF_UI_RebootRally_FloatingItem}
  NamedRerouteDeclaration(None=@n11.RGB) {name=Image 03 RGB, nodeColor=<rgba 0 0 0 1>}
  n12 = MaterialFunctionCall(
   Texture=TextureObjectParameter() {group=Image 04 Properties, parameterName=Texture 04, sortPriority=0}
   Offset=AppendVector(
    A=ScalarParameter() {group=Image 04 Properties, parameterName=Image 04 Offset X}
    B=ScalarParameter() {group=Image 04 Properties, parameterName=Image 04 Offset Y}
   )
   Scale=ScalarParameter() {defaultValue=0.5, group=Image 04 Properties, parameterName=Image 04 Scale}
   Flip X=StaticBoolParameter() {group=Image 04 Properties, parameterName=Image 04 Flip X}
   Rotation=ScalarParameter() {group=Image 04 Properties, parameterName=Image 04 Rotation}
   Base Speed=NamedRerouteUsage()
   Speed Modifier=Constant() {r=0.025}
   Range Min=NamedRerouteUsage()
   Range Max=NamedRerouteUsage()
   Animate=StaticBoolParameter() {defaultValue=true, parameterName=Animate}
  ) {materialFunction=&MF_UI_RebootRally_FloatingItem}
  NamedRerouteDeclaration(None=@n12.RGB) {name=Image 04 RGB, nodeColor=<rgba 0 0 0 1>}
  NamedRerouteDeclaration(None=@n8.Alpha) {name=Image 01 Mask, nodeColor=<rgba 0 0 0 1>}
  NamedRerouteDeclaration(None=@n10.Alpha) {name=Image 02 Mask, nodeColor=<rgba 0 0 0 1>}
  NamedRerouteDeclaration(None=@n11.Alpha) {name=Image 03 Mask, nodeColor=<rgba 0 0 0 1>}
  NamedRerouteDeclaration(None=@n12.Alpha) {name=Image 04 Mask, nodeColor=<rgba 0 0 0 1>}
  NamedRerouteUsage()
  NamedRerouteUsage()
  StaticSwitchParameter(True=Max(A=@n2, B=NamedRerouteUsage()), False=@n2) {parameterName=Display Additionnal Item}
 ]
}
