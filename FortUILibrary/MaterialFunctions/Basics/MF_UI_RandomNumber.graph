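# MF_UI_RandomNumber is a UI material function that returns a seeded pseudo-random number within a configurable min and max, defaulting to 0 and 1.
# It feeds floored texture coordinates and the seed into MF_UI_Random, so the value is constant within each integer UV cell.
# RemapValueRange scales that result from 0–1 to the requested range.
# It outputs both the raw float and a version rounded to the nearest integer.
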MaterialFunction /FortUILibrary/MaterialFunctions/Basics/MF_UI_RandomNumber
graph {
 properties {previewMaterialDomain=MD_UI}
 nodes [
  n1 = TextureCoordinate()
  n2 = FunctionInput() {
   bUsePreviewValueAsDefault=true
   description=Random Seed default set to 0. To change this you will change the random number generated
   inputName=RandomSeed
   inputType=FunctionInput_Scalar
   previewValue=<0 0 0 0>
  }
  n3 = FunctionInput() {
   bUsePreviewValueAsDefault=true
   description=Minimum range for random default set to 0.
   inputName=RandomNumberMin
   inputType=FunctionInput_Scalar
   previewValue=<0 0 0 0>
  }
  n4 = FunctionInput() {
   bUsePreviewValueAsDefault=true
   description=Maximum range for random default set to 1.
   inputName=RandomNumberMax
   inputType=FunctionInput_Scalar
   previewValue=<1 1 1 1>
  }
  n5 = MaterialFunctionCall(
   Input=MaterialFunctionCall!(UV=Floor(None=@n1), Seed=@n2) {materialFunction=&MF_UI_Random}.Result
   Input Low=Constant()
   Input High=Constant() {r=1}
   Target Low=@n3
   Target High=@n4
  ) {materialFunction=&RemapValueRange}
  FunctionOutput(None=@n5.Result) {description=This outputs a float that can have decimals, outputName=Float}
  FunctionOutput(None=Round(None=@n5.Result)) {description=This outputs the number rounded to the nearest int, outputName=Int, sortPriority=1}
 ]
}
