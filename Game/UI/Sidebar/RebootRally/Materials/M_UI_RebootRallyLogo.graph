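# Translucent UI material for the Reboot Rally logo, drawing an SDF icon with a cyan glow and animated CRT hologram-style effects.
# SmoothStep shapes the SDF into glow and edge, while delayed random jitter and a distortion term shift its UVs vertically.
# Sine-based scanlines and thin moving CRT lines combine with eased vertical gradients to modulate opacity.
# A panning bottom-cloud texture, distorted by a rotating wind flow map, forms a separate named Cloud layer.
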Material /Game/UI/Sidebar/RebootRally/Materials/M_UI_RebootRallyLogo
graph {
 properties {
  bUsedWithStaticMesh=false
  blendMode=BLEND_Translucent
  materialDomain=MD_UI
  pixelDepthOffsetMode=PDOM_Legacy
  usedShadingModels=Default Lit
 }
 nodes [
  n1 = MaterialFunctionCall(
   UVs=MaterialFunctionCall() {materialFunction=&MF_UI_UVScaling}.Scaled to Fit
   Texture Scale=ScalarParameter() {defaultValue=1.6, parameterName=Image Scale}
  ) {materialFunction=&ScaleUVsByCenter}
  n2 = MaterialFunctionCall!(RandomNumberMin=Constant(), RandomNumberMax=Constant() {r=0.02}, RandomSeed=Constant() {r=2}) {materialFunction=/FortUILibrary/MaterialFunctions/Basics/MF_UI_RandomNumber}
  n3 = MaterialFunctionCall(RandomNumberMin=Constant() {r=0.05}, RandomNumberMax=Constant() {r=2.5}) {materialFunction=&MF_UI_RandomNumber}
  n4 = Constant() {r=1}
  n5 = AppendVector(
   A=ScalarParameter() {parameterName=Image X Offset}
   B=LinearInterpolate(Alpha=MaterialFunctionCall!(Animation Duration=@n2.Float, Delay=@n3.Float, Reset to Default=@n4) {materialFunction=/FortUILibrary/MaterialFunctions/Time/MF_UI_AnimationDelay}.Result) {constB=0.04}
  )
  n6 = Constant()
  n7 = Multiply(A=NamedRerouteUsage(), B=ScalarParameter() {defaultValue=0.01, group=CRT Lines, parameterName=Distortion Amount, sliderMax=0.25})
  n8 = TextureSampleParameter2D(UVs=Add(A=Add(A=@n1.UVs, B=@n5), B=AppendVector(A=@n6, B=@n7))) {
   group=Icon
   parameterName=Icon Texture
   samplerType=SAMPLERTYPE_LinearGrayscale
   sortPriority=33
   texture=/Game/UI/Sidebar/RebootRally/Textures/RebootRally_Icon_SDF
  }
  n9 = SmoothStep(Value=@n8.RGB) {constMax=0.5}
  Add(A=LinearInterpolate(B=Constant4Vector() {constant=<rgba 0 0.6951 1 1>}.RGBA, Alpha=@n9))
  n10 = Floor(None=ScalarParameter() {defaultValue=52, group=Scanlines, parameterName=Scanline Count})
  n11 = MaterialFunctionCall() {materialFunction=&GetUserInterfaceUV}
  n12 = Frac(None=Multiply(A=Time(), B=ScalarParameter() {defaultValue=0.05, group=Scanlines, parameterName=Scanline Speed, sliderMax=0.25}))
  n13 = Multiply(A=Add(A=Add(A=@n11.Normalized UV, B=@n12)) {constB=2}) {constB=0.5}
  n14 = ComponentMask(None=Sine(None=Multiply(A=@n10, B=@n13))) {g=true}
  n15 = ScalarParameter() {defaultValue=0.05, group=Scanlines, parameterName=Scanline Thickness, sliderMax=0.1}
  MaterialExpressionRemap_0 = MaterialXRemap!(None=Saturate(None=Subtract(A=@n14, B=@n15) {constB=0.5}), InputLow, InputHigh=@n15, TargetLow, TargetHigh)
  NamedRerouteDeclaration(None=@MaterialExpressionRemap_0) {name=Scanlines, nodeColor=<rgba 0 0 0 1>}
  n16 = MaterialFunctionCall(UV=MaterialFunctionCall() {materialFunction=&GetUserInterfaceUV}.Normalized UV) {materialFunction=&MF_UI_UVScaling}
  NamedRerouteDeclaration(None=@n16.Scaled To Fill) {name=Scale to Fill, nodeColor=<rgba 0 0 0 1>}
  n17 = MaterialFunctionCall(UVs=@n16.Scaled To Fill, Texture Scale=Constant2Vector() {g=0.55, r=1}) {materialFunction=&ScaleUVsByCenter}
  n18 = ScalarParameter() {defaultValue=-0.4, parameterName=Bottom Cloud Y Position}
  NamedRerouteDeclaration(None=Panner(Coordinate=Add(A=@n17.UVs, B=@n18), Time=Time() {desc=Instruction to Move}) {speedX=0.025}) {name=scale down and move texture, nodeColor=<rgba 0 0 0 1>}
  n19 = Time()
  n20 = Constant() {r=0.02}
  n21 = MaterialFunctionCall(UVs=NamedRerouteUsage(), "Rotation Angle (0-1)"=Frac(None=Multiply(A=@n19, B=@n20))) {materialFunction=&CustomRotator}
  NamedRerouteDeclaration(None=MaterialFunctionCall!(
    UVs=NamedRerouteUsage()
    Texture Object=TextureObject() {samplerType=SAMPLERTYPE_Alpha, texture=/Game/UI/Sidebar/RebootRally/Textures/T_UI_cloudsBottom}
    Grayscale Flow Map=TextureSample(UVs=@n21.Rotated Values) {samplerType=SAMPLERTYPE_LinearColor, texture=/Game/Global/Textures/Default/Flowmaps/WindFlowMap}.RGB
    Flow Direction
    Flow Intensity=Constant() {r=0.2}
    Flow Speed=Constant() {r=-0.01}
   ) {materialFunction=/FortUILibrary/MaterialFunctions/Effects/MF_UI_GrayscaleFlowMap_New}.Distorted Texture) {name=Cloud, nodeColor=<rgba 0 0 0 1>}
  n22 = Multiply(A=Time(), B=ScalarParameter() {defaultValue=0.475, group=CRT Lines, parameterName=CRT Thin 1 Speed, sliderMax=1})
  n23 = MaterialFunctionCall() {materialFunction=&GetUserInterfaceUV}
  MaterialExpressionRemap_1 = MaterialXRemap(
   None=MaterialFunctionCall!(UV=Floor(None=@n23.Normalized UV), Seed=ScalarParameter() {defaultValue=4, parameterName=RandomSeed}) {materialFunction=&MF_UI_Random}.Result
   InputLow=Constant()
   InputHigh=Constant() {r=1}
   TargetLow=Constant()
   TargetHigh=Constant() {r=1}
  )
  n24 = Sine(None=Add(A=Panner(Coordinate=@n23.Normalized UV, Time=@n22) {speedY=0.52}, B=@MaterialExpressionRemap_1))
  n25 = ScalarParameter() {defaultValue=128, group=CRT Lines, parameterName=CRT Hardness}
  n26 = Multiply(A=Subtract(A=ComponentMask(None=@n24) {g=true}) {constB=0.99}, B=@n25)
  n27 = Multiply(A=Time(), B=ScalarParameter() {defaultValue=0.4, group=CRT Lines, parameterName=CRT Thin 2 Speed, sliderMax=1})
  n28 = Sine(None=Add(A=Panner(Coordinate=@n23.Normalized UV, Time=@n27) {speedY=0.5}, B=@MaterialExpressionRemap_1))
  n29 = Multiply(A=Subtract(A=ComponentMask(None=@n28) {g=true}) {constB=0.9}, B=@n25)
  NamedRerouteDeclaration(None=Add(A=Saturate(None=@n26), B=Saturate(None=@n29))) {name=CRT Lines, nodeColor=<rgba 0 0 0 1>}
  n30 = Max(A=SmoothStep(Value=@n8.RGB) {constMax=0.4, constMin=0.35}, B=Multiply(A=Power(Base=@n9)) {constB=0.5})
  n31 = NamedRerouteUsage()
  n32 = ScalarParameter() {defaultValue=0.05, group=Scanlines, parameterName=Scanline Opacity}
  n33 = Add(
   A=Multiply(A=NamedRerouteUsage(), B=ScalarParameter() {defaultValue=0.6, group=CRT Lines, parameterName=CRT Lines Opacity})
   B=OneMinus(None=Multiply(A=@n31, B=@n32) {constB=0.1})
  )
  n34 = MaterialFunctionCall("Value (0-1)"=ComponentMask(None=MaterialFunctionCall() {materialFunction=&GetUserInterfaceUV}.Normalized UV) {g=true}) {materialFunction=&MF_UI_EaseCurves}
  n35 = OneMinus(None=Multiply(A=LinearInterpolate(B=@n33, Alpha=@n34.EaseInQuint)) {constB=4})
  n36 = OneMinus(None=@n34.EaseInQuint)
  Saturate(None=LinearInterpolate(B=LinearInterpolate(B=@n30, Alpha=@n35), Alpha=Power(Base=@n36) {constExponent=4}))
  ScalarParameter() {parameterName=Image Y Offset}
 ]
}
