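# Sidebar social profile sub-panel widget that tells players when they or a friend have been rallied for Reboot Rally rewards.
# When the parent FortSocialProfileRebootRallyText fires DisplayText, the rich text block collapses if not rallied, otherwise it becomes visible.
# IsLocalUser selects between local-user and friend wording.
# A Reboot Rally style table and image decorator render the inline icon and highlight markup.
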WidgetBlueprint /Game/UI/Sidebar/RebootRally/SidebarProfileSubPanelRebootRallyEligible
graphs [EventGraph {nodes [
   n1 = Event DisplayText!() -> (OutputDelegate, then, bRallied:bool) {is=title K2Node_Event}
   > n2 = Branch(Condition=@n1.bRallied)
   n3 = Get RebootRallyEligibleText() -> (RebootRallyEligibleText:Common Rich Text Block Object Reference)
   > Set Visibility(Visibility=SelfHitTestInvisible, self=@n3)
   > SetText(
    self=Get RebootRallyEligibleText()
    InText=Select(
     Option 0:"Text (by ref)"=Get SocialUserText() -> (SocialUserText:Text)
     Option 1:"Text (by ref)"=Get LocalUserText() -> (LocalUserText:Text)
     Index=IsLocalUser!(self:Fort Social Profile Reboot Rally Text Object Reference) -> (ReturnValue:bool)
    ) -> (ReturnValue:"Text (by ref)")
   )
   Set Visibility(execute=@n2.else, Visibility=Collapsed, self=@n3)
   Event PreConstruct()
   Event Construct()
   Event Tick()
   Event OnFinishedProfileLoading() {is=title K2Node_Event}
   Event OnStartProfileLoading() {is=title K2Node_Event}
  ]}]
variables [
 LocalUserText {
  type=string
  default="You've been <img id=\"RebootRally\"/> <HighlightBlue14>rallied!</> Play with friends to earn cosmetic rewards in Reboot Rally!"
 }
 SocialUserText {
  type=string
  default="Your friend has been <img id=\"RebootRally\"/> <HighlightBlue14>rallied!</> Play together to earn cosmetic rewards in Reboot Rally!"
 }
]
widgets {
 info {
  inheritedWidgetCount=1
  parentClass=/Script/FortniteUI.FortSocialProfileRebootRallyText
  rootWidgetClass=&CommonRichTextBlock
  widgetCount=2
 }
 widgets [
  RebootRallyEligibleText {
   widgetClassPath=&CommonRichTextBlock
   bIsVariable=true
   widget=RebootRallyEligibleText
   properties {
    autoWrapText=true
    bIsScrollingEnabled=false
    bOverrideDefaultStyle=true
    justification=Center
    mobileTextBlockScale=1.15
    text="[Reboot_Rally_Eligible_Text]"
    textStyleSet=&/DT_UI_RebootRallyRichTextStyles
    visibility=Collapsed
    decoratorClasses [&/BP_UI_RebootRallyRichImageDecorator.BP_UI_RebootRallyRichImageDecorator_C]
    defaultTextStyleOverride {
     overflowPolicy=Clip
     transformPolicy=ToUpper
     colorAndOpacity {colorUseRule=UseColor_Specified, specifiedColor=<rgba 1 1 1 1>}
     font {
      bForceMonospaced=false
      bMaterialIsStencil=false
      fontMaterial=null
      fontObject=&BurbankSmall
      letterSpacing=0
      monospacedWidth=1
      size=14
      skewAmount=0
      typefaceFontName=Bold
      outlineSettings {
       bApplyOutlineToDropShadows=false
       bMiteredCorners=false
       bSeparateFillAlpha=false
       outlineMaterial=null
       outlineSize=0
       outlineColor=<rgba 0 0 0 1>
      }
     }
     highlightColor {colorUseRule=UseColor_Specified, specifiedColor=<rgba 0 0 0 1>}
     highlightShape {
      bIsDynamicallyLoaded=false
      drawAs=Image
      imageType=NoImage
      mirroring=NoMirror
      resourceName=None
      resourceObject=null
      tiling=NoTile
      imageSize=<32 32>
      margin=<ltrb 0 0 0 0>
      outlineSettings {
       bUseBrushTransparency=false
       roundingType=HalfHeightRadius
       width=0
       color {colorUseRule=UseColor_Specified, specifiedColor=<rgba 0 0 0 0>}
       cornerRadii=<0 0 0 0>
      }
      tintColor {colorUseRule=UseColor_Specified, specifiedColor=<rgba 1 1 1 1>}
      uVRegion {bIsValid=false, max=<0 0>, min=<0 0>}
     }
     selectedBackgroundColor {colorUseRule=UseColor_Foreground, specifiedColor=<rgba 1 0 1 1>}
     shadowColorAndOpacity=<rgba 0 0 0 1>
     shadowOffset=<0 0>
     strikeBrush {
      bIsDynamicallyLoaded=false
      drawAs=Image
      imageType=NoImage
      mirroring=NoMirror
      resourceName=None
      resourceObject=null
      tiling=NoTile
      imageSize=<32 32>
      margin=<ltrb 0 0 0 0>
      outlineSettings {
       bUseBrushTransparency=false
       roundingType=HalfHeightRadius
       width=0
       color {colorUseRule=UseColor_Specified, specifiedColor=<rgba 0 0 0 0>}
       cornerRadii=<0 0 0 0>
      }
      tintColor {colorUseRule=UseColor_Specified, specifiedColor=<rgba 1 1 1 1>}
      uVRegion {bIsValid=false, max=<0 0>, min=<0 0>}
     }
     underlineBrush {
      bIsDynamicallyLoaded=false
      drawAs=Image
      imageType=NoImage
      mirroring=NoMirror
      resourceName=None
      resourceObject=null
      tiling=NoTile
      imageSize=<32 32>
      margin=<ltrb 0 0 0 0>
      outlineSettings {
       bUseBrushTransparency=false
       roundingType=HalfHeightRadius
       width=0
       color {colorUseRule=UseColor_Specified, specifiedColor=<rgba 0 0 0 0>}
       cornerRadii=<0 0 0 0>
      }
      tintColor {colorUseRule=UseColor_Specified, specifiedColor=<rgba 1 1 1 1>}
      uVRegion {bIsValid=false, max=<0 0>, min=<0 0>}
     }
    }
   }
  }
  SubPanel_Loading {widgetClassPath=/Script/FortniteUI.FortSocialProfileSubPanel_Loading, bInherited=true, widget=null}
 ]
}
