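# Reboot Rally sidebar icon widget, derived from the native RebootRallyIcon class.
# A clipped CommonBorder centers a scaled-up Image that draws the M_UI_RebootRally_Icon material.
# The Highlighted animation holds constant Icon and Glow color material parameters, while Default is an empty animation.
# The event graph holds only unconnected event stubs, and nothing in this Blueprint plays the animations.
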WidgetBlueprint /Game/UI/Sidebar/RebootRally/WBP_RebootRallyIcon
animations [
 Default {binding_count=0, length_seconds=5, object=&self:Default, sequence {bindings []}}
 Highlighted {
  binding_count=2
  length_seconds=5
  object=&self:Highlighted
  sequence {bindings [
    Image_RebootRally {
     id=#1
     possessed_class=Image
     tracks [{
       type=MovieSceneWidgetMaterialTrack
       sections [{
         type=MovieSceneParameterSection
         end_frame=100
         start_frame=0
         channels [
          Icon ColorR {type=MovieSceneScriptingFloatChannel, const [0,1]}
          Icon ColorG {type=MovieSceneScriptingFloatChannel, const [0.15,1]}
          Icon ColorB {type=MovieSceneScriptingFloatChannel, const [1,1]}
          Icon ColorA {type=MovieSceneScriptingFloatChannel, const [1,1]}
          Glow ColorR {type=MovieSceneScriptingFloatChannel, const [0,1]}
          Glow ColorG {type=MovieSceneScriptingFloatChannel, const [0.5774,1]}
          Glow ColorB {type=MovieSceneScriptingFloatChannel, const [1,1]}
          Glow ColorA {type=MovieSceneScriptingFloatChannel, const [0,1]}
         ]
        }]
      }]
    }
    BorderSlot_0 {id=#2, parent_id=#1, possessed_class=BorderSlot, tracks []}
   ]}
 }
]
graphs [EventGraph {nodes [
   Event PreConstruct()
   Event Construct()
   Event Tick()
  ]}]
variables []
widgets {
 info {inheritedWidgetCount=0, parentClass=/Script/FortniteUI.RebootRallyIcon, rootWidgetClass=&CommonBorder, widgetCount=2}
 widgets [
  Border {
   widgetClassPath=&CommonBorder
   widget=Border
   properties {clipping=ClipToBounds, horizontalAlignment=HAlign_Center, verticalAlignment=VAlign_Center, padding=<ltrb 0 0 0 0>}
  }
  Image_RebootRally {
   widgetClassPath=&Image
   bIsVariable=true
   parent=Border
   widget=Image_RebootRally
   properties {
    brush {
     bIsDynamicallyLoaded=false
     drawAs=Image
     imageType=NoImage
     mirroring=NoMirror
     resourceName=None
     resourceObject=/Game/UI/Sidebar/RebootRally/Materials/M_UI_RebootRally_Icon
     tiling=NoTile
     imageSize=<32 32>
     margin=<ltrb 0 0 0 0>
     outlineSettings {
      bUseBrushTransparency=false
      roundingType=HalfHeightRadius
      width=0
      color {colorUseRule=UseColor_Specified, specifiedColor=<rgba 0 0 0 0>}
      cornerRadii=<0 0 0 0>
     }
     tintColor {colorUseRule=UseColor_Specified, specifiedColor=<rgba 1 1 1 1>}
     uVRegion {bIsValid=false, max=<0 0>, min=<0 0>}
    }
    renderTransform {angle=0, scale=<1.5 1.5>, shear=<0 0>, translation=<0 0>}
   }
   slot {class=BorderSlot, properties {horizontalAlignment=HAlign_Center, verticalAlignment=VAlign_Center, padding=<ltrb 0 0 0 0>}}
  }
 ]
}
